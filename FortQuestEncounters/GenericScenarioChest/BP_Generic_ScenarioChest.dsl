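; Generic scenario chest Blueprint for quest encounters, extending a parent chest class with custom visual effects.
; Its construction script tints the glow and smoke sheet materials red-pink through SetFXColor, then sets the custom Niagara effects array to the rare chest visuals.
; SpawnChestOpenVFX spawns a Niagara burst at the static mesh's world transform.
; BeginPlay only calls the parent, while the overlap and Tick events are empty.
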(event EventBeginPlay
  (|CallParentFunction))

(event Collision|EventActorBeginOverlap
  (params (OtherActor :type Actor%20Object%20Reference)))

(event EventTick
  (params (DeltaSeconds :type Float%20%28single-precision%29)))

(event EventSpawnChestOpenVFX
  (Niagara|SpawnSystematLocation "/FortQuestEncounters/GenericScenarioChest/FX/NS_GenericScenario_ChestBurst.NS_GenericScenario_ChestBurst" (.location (Transformation|GetWorldTransform (Variables|Debug|GetStaticMeshComponent))) (.rotation (Transformation|GetWorldTransform (Variables|Debug|GetStaticMeshComponent)))))

(fn SetFXColor
  (bind _fx_glow___smoke_sheet (Variables|Default|GetFX%20Glow%20&%20Smoke%20Sheet))
  (Rendering|Material|SetScalarParameterValueonMaterials _fx_glow___smoke_sheet "WPO_HideMesh" 1.0)
  (Rendering|Material|SetScalarParameterValueonMaterials _fx_glow___smoke_sheet "Corona_Opacity" 2e+01)
  (Rendering|Material|SetScalarParameterValueonMaterials _fx_glow___smoke_sheet "Glow_Opacity" 7.0)
  (Rendering|Material|SetVectorParameterValueonMaterials _fx_glow___smoke_sheet "Corona_Colour" "1.000000,0.000000,0.000000")
  (Rendering|Material|SetVectorParameterValueonMaterials _fx_glow___smoke_sheet "Glow_Colour" "1.000000,0.000000,0.000000")
  (Rendering|Material|SetScalarParameterValueonMaterials _fx_glow___smoke_sheet "UseBaseColor" 0.0)
  (Rendering|Material|SetColorParameterValueonMaterials _fx_glow___smoke_sheet "SecColor" "(R=1.000000,G=0.000000,B=0.270417,A=1.000000)")
  (Rendering|Material|SetColorParameterValueonMaterials _fx_glow___smoke_sheet "SpriteColor_Low" "(R=1.000000,G=0.025792,B=0.298992,A=1.000000)"))

(fn ConstructionScript
  (|CallParentFunction)
  (CallFunction|SetFXColor)
  (Variables|Default|SetCustom%20Niagara%20Effects (Utilities|Array|MakeArray (Variables|Default|GetRare%20Chest%20Visuals))))
